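; Chain Whip secondary hookshot ability (v2) that drives the hookshot reticle while active.
; On non-authority clients, a hotfix-tunable timer loop checks for pawns under the reticle, falling back a tick later to a targeting-preset trace to spread cost.
; Results set the weapon targeting component's target type, briefly caching the last homing target to ease aiming.
; It also registers spawned projectiles with the weapon and reports throws server-side to a GameState component.
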(event Ability|EventActivateAbility
  ;[0.00519 0.2552 0.0659] ActivateAbility: Client only: Perform targeting and update the targeting component with the latest information.
  ; The targeting component in turn updates the crosshair.
  (|CallParentFunction)
  (Ability|HasAuthority
    (:False
      (Variables|Transient|SetCachedShouldCheckForTargetsUnderReticle (ScalableFloat|GetValueasBool (Variables|Hotfix|GetHotfix_ReticleUnderEnable)))
      (Variables|Transient|SetCachedUnderReticleSize (ScalableFloat|GetValueatLevel (Variables|Hotfix|GetHotfix_ReticleUnderSize)))
      (CallFunction|SetHookshotTargetingEnabled true))))

(event Custom|SetHookshotTargetingEnabled
  (params (bEnabled :type Boolean))
  ;[0.125 0.01902 0.09411] HookshotTargeting: Perform targeting and update the targeting component with the latest information.
  ; The targeting component in turn updates the crosshair.
  ;;[0 0.5208 0.1542] SetHookshotTargetingEnabled: Client only. If enabled, start performing hookshot targeting trace.
  (Ability|HasAuthority
    (:False
      (if bEnabled
        (CallFunction|PerformHookshotTargetingTrace) ; Self calling function that executes the targeting task for the hookshot ability.
        (else
          (Utilities|Time|ClearandInvalidateTimerbyHandle (Variables|Transient|GetHookshotTargetingTraceTimerHandle))
          (bind _asynctask_hookshottargeting (Variables|Transient|GetAsyncTask_HookshotTargeting)
            (:then
              (GameplayTasks|EndTask _asynctask_hookshottargeting))
            (:else)))))))

(event Custom|PerformHookshotTargetingTrace
  (bind _hotfix_reticletracetargetingupdatefreq (Variables|Hotfix|GetHotfix_ReticleTraceTargetingUpdateFreq))
  (bind _returnvalue (ScalableFloat|GetValueatLevel _hotfix_reticletracetargetingupdatefreq))
  ;[0.125 0.01902 0.09411] HookshotTargeting: Perform targeting and update the targeting component with the latest information.
  ; The targeting component in turn updates the crosshair.
  ;;[0.08033 0.1667 0.1449] PerformHookshotTargetingTrace: Self calling function that executes the targeting task for the hookshot ability.
  ;; First check for any valid pawns to target under reticle, if none found - perform a trace and use that result instead.
  (Ability|HasAuthority
    (:False
      (bind (_bvalidtarget _targetdata) (Default|GetTargetPawnUnderReticle))
      (if _bvalidtarget
        (Default|UpdateTargetingComponent _targetdata)
        (else
          (Utilities|FlowControl|DelayUntilNextTick) ; GetTargetPawnUnderReticle() and PerformTargetingAbilityTask() are both expensive, so we separate them by 1 framde.
          (bind (_asynctaskproxy _targetdata_2) (Ability|Tasks|PerformTargetingAbilityTask "Task_HookshotTargeting" (Variables|Targeting|GetTargetingPresetHookshotTargeting) false false (ScalableFloat|GetValueasBool (Variables|Hotfix|GetHotfix_ReticleTraceUseAsync)))
            (:then
              (Variables|Transient|SetAsyncTask_HookshotTargeting _asynctaskproxy))
            (:Targeted
              _)
            (:Cancelled
              _))
          (Default|UpdateTargetingComponent _targetdata_2)))
      (if (> _returnvalue 0)
        (bind _returnvalue_2 (Utilities|Time|SetTimerbyEvent (EventDispatchers|CreateEvent :!event Custom|PerformHookshotTargetingTrace) _returnvalue false true))
        (Variables|Transient|SetHookshotTargetingTraceTimerHandle _returnvalue_2)
        (else
          (bind _returnvalue_3 (Utilities|Time|SetTimerforNextTickbyEvent (EventDispatchers|CreateEvent :!event Custom|PerformHookshotTargetingTrace)))
          (Variables|Transient|SetHookshotTargetingTraceTimerHandle _returnvalue_3))))))

(event Ability|EventOnEndAbility
  (params (bWasCancelled :type Boolean))
  ;[0.15 0.08962 0.01766 0.5] EndAbility: Cleanup.
  (Utilities|FlowControl|Sequence
    (:then_0
      (CallFunction|SetHookshotTargetingEnabled false)
      (Default|ClearLastTargetedValidHomingTarget)
      (Utilities|Time|ClearandInvalidateTimerbyHandle (Variables|Transient|GetClearLastTargetedValidHomingTargetTimerHandle))
      (bind _targetingcomponent (Variables|Transient|GetTargetingComponent)
        (:then
          (Reticle|Targeting|SetTargetType _targetingcomponent "NewEnumerator0"))
        (:else)))
    (:then_1
      (|CallParentFunction bWasCancelled))))

(fn AnalyticsHookshotThrown
  (bind _returnvalue (Fortnite|Athena|GameState|GetGameStateAthena))
  (bind _returnvalue_2 (Actor|GetComponentbyClass _returnvalue "/SunRoseChainWhipGameplay/Items/ChainWhip/B_ChainWhipGameStateComponent.B_ChainWhipGameStateComponent_C"))
  ;[0.03924 0.11 0.1146] AnalyticsHookshotUsed: Let the game state component know that a hookshot was used.
  (|CallParentFunction)
  (Ability|HasAuthority
    (:True
      (Utilities|IsValid _returnvalue
        (:Is%20Valid
          (Utilities|IsValid _returnvalue_2
            (:Is%20Valid
              (Class|BChainWhipGameStateComponent|NotifyHookshotThrown _returnvalue_2 (CallFunction|GetAvatarasFortPlayerPawn)))))))))

(fn ClearLastTargetedValidHomingTarget
  ;[0.03475 0.0393 0.1823] Clear the last targeted valid homing target.
  ; This is called after the short cached duration is complete.
  (Variables|Transient|SetLastTargetedValidHomingTarget 0))

(fn GetTargetPawnUnderReticle
  ;[0.03448 0.01049 0.04167] GetTargetPawnUnderReticle: Find the best target to use by looking at potential target pawns under the reticle; return the best result as the target data.
  (Utilities|FlowControl|Sequence
    (:then_0
      (if (Variables|Transient|GetCachedShouldCheckForTargetsUnderReticle)
        (bind _outtargetpawn (AutoAimWeapon|FindBestAutoAimTarget (CallFunction|GetAvatarasFortPlayerPawn) (Variables|Transient|GetCachedUnderReticleSize) (Variables|Transient|GetCachedHookshotDistance)))
        (Utilities|IsValid _outtargetpawn
          (:Is%20Valid
            (return true (Ability|TargetData|AbilityTargetDatafromHitResult (Collision|MakeHitResult true false 0.0 0.0 "0, 0, 0" "0, 0, 0" "0,0,1" "0,0,1" 0 _outtargetpawn)))))))
    (:then_1
      (return false 0))))

(fn HandleHookshotProjectileSpawned
  (params (HookshotPrj :type B%20Prj%20Hookshot%20Object%20Reference))
  ;[0.02059 0.15 0.1325 0.5] Hookshot projectile spawned. Let the weapon track it.
  (|CallParentFunction HookshotPrj)
  (bind _asb_prj_chain_whip_hookshot_v_2 (Utilities|Casting|CastToB_Prj_ChainWhip_Hookshot_v2 HookshotPrj))
  (bind _asb_chain_whip (Utilities|Casting|CastToB_ChainWhip (Ability|GetCurrentSourceWeapon)))
  (Hookshot|SetHookshotProjectile _asb_chain_whip _asb_prj_chain_whip_hookshot_v_2))

(fn UpdateTargetingComponent
  (params (TargetData :type Gameplay%20Ability%20Target%20Data%20Handle%20Structure))
  (bind _returnvalue (Ability|GetCurrentSourceWeapon))
  (bind _returnvalue_2 (Ability|TargetData|GetHitResultfromTargetData TargetData))
  (bind (_bblockinghit _binitialoverlap _time _distance _location _impactpoint _normal _impactnormal _physmat _hitactor) (Collision|BreakHitResult _returnvalue_2))
  ;[0.03183 0.1354 0.1246] UpdateTargetingComponent: Run the trace to find the potential target that the hookshot ability is going to land on and update the reticle through the target component.
  (Variables|Transient|GetTargetingComponent
    (:then
      _)
    (:else
      (Utilities|IsValid _returnvalue
        (:Is%20Valid
          (bind _output_get (Variables|Transient|SetTargetingComponent (Actor|GetComponentbyClass _returnvalue "/SunRoseChainWhipGameplay/Items/ChainWhip/B_ChainWhipTargetingComponent.B_ChainWhipTargetingComponent_C")))
          (Utilities|IsValid _output_get
            (:Is%20Valid
              _))))))
  (bind _targetactor (Homing|GetHomingTargetfromTargetData TargetData))
  (Utilities|IsValid _targetactor
    (:Is%20Valid
      ;;[0.1615 0.09374 0.11] When a valid homing target is found, we cache it for a short duration to make it easier to track/aim at that target.
      (LocalVariables|SetTargetType "NewEnumerator2")
      (bind _output_get_2 (LocalVariables|SetCacheLastValidHomingTargetDuration (ScalableFloat|GetValueatLevel (Variables|Hotfix|GetHotfix_CacheLastValidHomingTargetDuration))))
      (if (> _output_get_2 0)
        (Variables|Transient|SetLastTargetedValidHomingTarget _targetactor)
        (bind _returnvalue_3 (Utilities|Time|SetTimerbyEvent (EventDispatchers|CreateEvent :!event ClearLastTargetedValidHomingTarget) (LocalVariables|GetCacheLastValidHomingTargetDuration)))
        (Variables|Transient|SetClearLastTargetedValidHomingTargetTimerHandle _returnvalue_3)
        (else
          _)))
    (:Is%20Not%20Valid
      ;;[0.03108 0.04441 0.1823] IF we did not find a valid homing target in the current trace, but our LastTargetedValidHomingTarget is still valid, consider that Target instead.
      (bind _bvalid (Homing|IsValidHomingTarget (Variables|Transient|GetLastTargetedValidHomingTarget)))
      (if _bvalid
        (LocalVariables|SetTargetType "NewEnumerator2")
        (else
          (Utilities|IsValid _hitactor
            (:Is%20Valid
              (LocalVariables|SetTargetType "NewEnumerator1"))
            (:Is%20Not%20Valid
              _))))))
  (Reticle|Targeting|SetTargetType (Variables|Transient|GetTargetingComponent) (LocalVariables|GetTargetType))
  (Development|PrintString (Utilities|String|Append "HookshotTargeting: " (select (Utilities|IsValid _targetactor :!class /Script/Engine.KismetSystemLibrary) (Utilities|GetDisplayName _hitactor) "No Target") " [" (Utilities|String|EnumtoString (LocalVariables|GetTargetType)) "]") true false "(R=1.000000,G=0.482353,B=0.000000,A=1.000000)" 1.0 "Hookshot_UpdateTargetingComponent"))
