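; Gameplay ability for the Chain Whip's secondary hookshot pull (v2), adding analytics reporting on top of its parent ability.
; Its only active logic overrides NotifyAnalyticsSuccessfulHookshot, calling the parent before an authority check.
; With authority, it passes the instigator pawn, target character and source weapon's hookshot projectile to B_ChainWhipGameStateComponent on the Athena game state.
; The activate and end events are disabled stubs.
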(event Ability|EventActivateAbility
  (|CallParentFunction) ; This node is disabled and will not be called.\nDrag off pins to build functionality.
)

(event Ability|EventOnEndAbility
  (params (bWasCancelled :type Boolean))
  (|CallParentFunction bWasCancelled) ; This node is disabled and will not be called.\nDrag off pins to build functionality.
)

(fn NotifyAnalyticsSuccessfulHookshot
  (bind _returnvalue (Fortnite|Athena|GameState|GetGameStateAthena))
  (bind _returnvalue_2 (Actor|GetComponentbyClass _returnvalue "/SunRoseChainWhipGameplay/Items/ChainWhip/B_ChainWhipGameStateComponent.B_ChainWhipGameStateComponent_C"))
  ; Report analytics for a successful hookshot.
  (|CallParentFunction)
  (Ability|HasAuthority
    (:True
      (Utilities|IsValid _returnvalue
        (:Is%20Valid
          (Utilities|IsValid _returnvalue_2
            (:Is%20Valid
              (bind _asb_chain_whip (Utilities|Casting|CastToB_ChainWhip (Ability|GetCurrentSourceWeapon)))
              (Class|BChainWhipGameStateComponent|NotifySuccessfulHookshot _returnvalue_2 (Variables|Transient|GetInstigatorPlayerPawn) (Variables|Transient|GetTargetCharacter) (Hookshot|GetHookshotProjectile _asb_chain_whip)) ; Analytics
            )))))))
